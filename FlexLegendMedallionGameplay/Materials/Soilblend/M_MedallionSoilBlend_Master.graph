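# Master material for medallion soil-blend effects on character meshes, rendering a glassy, smoky haunted orb with a skull face inside.
# A camera-facing virtual plane maps the skull color and depth, using parallax occlusion on High/Epic and panning noise distortion.
# A custom HLSL raymarch adds flowing internal smoke, while Fresnel, matcap fake refraction and two-color tinting simulate glass.
# Base color and emissive results pass through shared character PreFX/PostFX and emissive-adjustment functions.
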Material /FlexLegendMedallionGameplay/Materials/Soilblend/M_MedallionSoilBlend_Master
graph {
 properties {
  bUseMaterialAttributes=true
  bUsedWithClothing=true
  bUsedWithMeshDeformer=true
  bUsedWithNanite=true
  bUsedWithNiagaraMeshParticles=true
  bUsedWithSkeletalMesh=true
  pixelDepthOffsetMode=PDOM_Legacy
  usedShadingModels=Default Lit
 }
 nodes [
  n1 = Constant() {r=0.05}
  n2 = ScalarParameter() {defaultValue=0.35, group=SoilHaunt, parameterName=Specular}
  n3 = NamedRerouteUsage()
  n4 = NamedRerouteUsage()
  n5 = Constant() {r=0.35}
  n6 = Constant() {r=1.5}
  n7 = NamedRerouteUsage()
  n8 = LinearInterpolate(A=Multiply(A=@n4, B=LinearInterpolate(A=@n5, B=@n6, Alpha=@n7) {constB=1.35}) {constB=0.1}, B=@n4, Alpha=NamedRerouteUsage())
  n9 = LinearInterpolate(A=Constant() {r=0.1}, B=Constant() {r=0.05}, Alpha=NamedRerouteUsage())
  n10 = TextureSample(UVs=Multiply(A=TextureCoordinate()) {constB=4}) {texture=&/T_Fake_refraction}
  n11 = Constant() {r=0.75}
  n12 = NamedRerouteUsage()
  n13 = NamedRerouteUsage()
  n14 = MaterialFunctionCall() {materialFunction=&MF_CharacterParent_2023_PreFX}
  MaterialFunctionCall(
   Material=MakeMaterialAttributes!(
    BaseColor=@n13
    Metallic=ScalarParameter() {defaultValue=0.1, group=SoilHaunt, parameterName=Metallic}
    Specular=LinearInterpolate(A=@n1, B=@n2, Alpha=@n3)
    Roughness=LinearInterpolate(A=@n9, Alpha=@n10.G) {constB=0.5}
    Anisotropy
    EmissiveColor=MaterialFunctionCall(Emissive=@n8) {materialFunction=&MF_FX_EmissiveAdjustments}.Result
    Opacity
    OpacityMask
    Normal
    Tangent
    WorldPositionOffset
    SubsurfaceColor
    ClearCoat
    ClearCoatRoughness
    AmbientOcclusion=LinearInterpolate(A=@n11, Alpha=@n12)
    Refraction
    CustomizedUVs_0
    CustomizedUVs_1
    CustomizedUVs_2
    CustomizedUVs_3
    CustomizedUVs_4
    CustomizedUVs_5
    CustomizedUVs_6
    CustomizedUVs_7
    PixelDepthOffset
    ShadingModel
    Displacement
   )
   M=@n14.M
   Pre Skinned Local Position=@n14.M
   Distance Blend=@n14.M
   Fresnel=@n14.M
  ) {materialFunction=&MF_CharacterParent_2023_PostFX}
  n15 = LinearInterpolate(
   A=NamedRerouteUsage()
   B=Constant() {r=0.5}
   Alpha=Saturate(None=TextureSample() {texture=&T_Fake_refraction}.B)
  )
  n16 = Constant() {r=1}
  n17 = VertexNormalWS()
  n18 = Clamp(None=Multiply(A=Fresnel(ExponentIn=@n16, Normal=@n17), B=ScalarParameter() {defaultValue=2, group=SoilHaunt, parameterName=Glass thickness}))
  n19 = LinearInterpolate(
   A=Constant() {r=0.75}
   B=LinearInterpolate(
    A=ScalarParameter() {defaultValue=0.2, group=SoilHaunt, parameterName=Fresnel Inner Transparency}
    B=ScalarParameter() {defaultValue=0.7, group=SoilHaunt, parameterName=Fresnel Outer Transparency}
    Alpha=NamedRerouteDeclaration(None=@n18) {name=Fresnel, nodeColor=<rgba 0.1016 1 0 1>}
   )
   Alpha=NamedRerouteUsage()
  )
  n20 = VectorParameter() {group=SoilHaunt, parameterName=Dark_color, defaultValue=<rgba 0.0067 0.09 0.0792 1>}
  n21 = VectorParameter() {group=SoilHaunt, parameterName=Bright_Color, defaultValue=<rgba 0.1468 1 0.7473 1>}
  n22 = Constant() {r=1.1}
  n23 = Constant() {r=10}
  n24 = LinearInterpolate(
   A=Multiply(A=@n21.RGB) {constB=0.5}
   B=Multiply(A=@n21.RGB, B=ScalarParameter() {defaultValue=1.9, group=SoilHaunt, parameterName=Internal luminance multiple}) {constB=10}
   Alpha=LinearInterpolate(
    A=Constant()
    Alpha=MaterialFunctionCall(Brightness=@n22, Exponent=@n23) {materialFunction=&MF_Fresnel_FX}.Fresnel
   )
  )
  n25 = NamedRerouteUsage()
  n26 = MaterialFunctionCall!(
   Heightmap Texture=TextureObjectParameter() {group=SoilHaunt, parameterName=MaskDepthTex, texture=&/T_skull_face_Depth}
   Height Ratio=Constant() {r=0.06}
   Min Steps=Constant() {r=9}
   Max Steps=Constant() {r=9}
   UVs=NamedRerouteUsage()
   Heightmap Channel=Constant4Vector() {constant=<rgba 1 0 0 0>}.RGBA
   Reference Plane
   -----------------------------
   Use World Coordinates
   Specify Manual Texture Size
   Manual Texture Size
   ------------------------------
   "Render Shadows (Occlusion Mapping)"=StaticBool()
   Light Vector=AtmosphericLightVector()
   Shadow Steps
   Shadow Penumbra
   Transform To VertexNormal
  ) {materialFunction=/Engine/Functions/Engine_MaterialFunctions01/Texturing/ParallaxOcclusionMapping}
  n27 = MaterialFunctionCall(Vector or UVs=Add(A=NamedRerouteUsage()) {constB=-0.5}) {materialFunction=&VectorToRadialValue}
  n28 = TextureCoordinate()
  n29 = TextureSample!(
   UVs=Panner(Coordinate=@n27.Radial Coordinates) {speedX=0.002, speedY=-0.05}
   Tex
   "DDX(UVs)"=DDX(Value=@n28)
   "DDY(UVs)"=DDY(Value=@n28)
   Apply View MipBias
  ) {mipValueMode=TMVM_Derivative, texture=&AtmosphericCloudNoise01}
  n30 = NamedRerouteUsage()
  n31 = Constant() {r=0.08}
  n32 = Constant() {r=0.05}
  n33 = TextureSampleParameter2D(UVs=NamedRerouteUsage()) {
   group=SoilHaunt
   parameterName=MaskDepthTex
   samplerSource=SSM_Clamp_WorldGroupSettings
   texture=&T_skull_face_Depth
  }
  n34 = OneMinus(None=@n33.B)
  n35 = Add(
   A=ShadingPathSwitch(Default=QualitySwitch(Default=@n25, Low=@n25, High=@n26.Parallax UVs, Medium=@n25, Epic=@n26.Parallax UVs), Mobile=@n25)
   B=Multiply(A=LinearInterpolate(A=@n29.R, Alpha=@n30) {constB=0}, B=LinearInterpolate(A=@n31, B=@n32, Alpha=@n34) {constB=0.01}) {constB=0.1}
  )
  n36 = TextureSampleParameter2D(UVs=@n35) {
   group=SoilHaunt
   parameterName=MaskColor
   samplerSource=SSM_Clamp_WorldGroupSettings
   texture=&/T_skull_face_Color
  }
  n37 = NamedRerouteUsage()
  n38 = Multiply(A=NamedRerouteUsage(), B=Constant() {r=1.5}) {constB=2}
  n39 = TextureSampleParameter2D(UVs=@n35) {
   group=SoilHaunt
   parameterName=MaskDepthTex
   samplerSource=SSM_Clamp_WorldGroupSettings
   texture=&T_skull_face_Depth
  }
  n40 = LinearInterpolate(B=Constant() {r=0.25}, Alpha=NamedRerouteDeclaration(None=OneMinus(None=@n39.B)) {name=Emissive Mask, nodeColor=<rgba 0 1 0.6016 1>}) {constA=0.1}
  n41 = LinearInterpolate(
   A=Multiply(A=@n36.RGB, B=Constant() {r=0.5}) {constB=0.5}
   B=Power(Base=Saturate(None=@n37), Exp=ScalarParameter() {defaultValue=0.85, group=SoilHaunt, parameterName=height_basecolor_exponent})
   Alpha=Saturate(None=Add(A=@n38, B=@n40) {constB=0.35})
  )
  n42 = Multiply(A=Multiply(A=OneMinus(None=@n19), B=LinearInterpolate(A=@n20.RGB, B=@n24, Alpha=@n41)), B=Constant() {r=0.75})
  NamedRerouteDeclaration(None=LinearInterpolate(A=Multiply(A=@n15, B=@n42), B=@n42, Alpha=ScalarParameter() {defaultValue=0.65, parameterName="strength of the  Surface smoke"}) {constAlpha=0.35}) {name=BaseColor, nodeColor=<rgba 0 1 0.5547 1>}
  n43 = TextureCoordinate()
  n44 = Panner(Coordinate=@n43) {speedX=0.04, speedY=0.02}
  n45 = TextureSample!(UVs=Panner(Coordinate=Multiply(A=@n43) {constB=2}) {speedX=0.1, speedY=0.06}, Tex, Apply View MipBias) {
   samplerSource=SSM_Wrap_WorldGroupSettings
   samplerType=SAMPLERTYPE_LinearGrayscale
   texture=/Game/Global/Textures/Default/Flowmaps/T_StormSwirlMask
  }
  n46 = Multiply(A=@n45.R) {constB=0.04}
  NamedRerouteDeclaration(None=TextureSample(UVs=Add(A=@n44, B=@n46)) {samplerSource=SSM_Wrap_WorldGroupSettings, texture=&AtmosphericCloudNoise01}.R) {name=Surface smoke, nodeColor=<rgba 1 0 0.5391 1>}
  MaterialExpressionMultiply_33 = Multiply(A=NamedRerouteUsage(), B=ScalarParameter() {defaultValue=0.43, group=SoilHaunt, parameterName=tile})
  MaterialExpressionTransform_2 = Transform(None=CameraVectorWS()) {transformSourceType=TRANSFORMSOURCE_World, transformType=TRANSFORM_Tangent}
  MaterialExpressionConstant_24 = Constant() {r=16}
  MaterialExpressionScalarParameter_24 = ScalarParameter() {defaultValue=0.0105, group=SoilHaunt, parameterName=step_dist}
  MaterialExpressionTextureObject_1 = TextureObject() {texture=&AtmosphericCloudNoise01}
  MaterialExpressionVertexNormalWS_4 = VertexNormalWS()
  MaterialExpressionTime_4 = Time()
  MaterialExpressionComponentMask_2 = ComponentMask(None=VectorParameter() {parameterName=heightspeed, defaultValue=<rgba -0.02 0 0 1>}.RGB) {g=true, r=true}
  MaterialExpressionComponentMask_3 = ComponentMask(None=VectorParameter() {parameterName=distortspeed, defaultValue=<rgba -0.04 0 0 1>}.RGB) {g=true, r=true}
  MaterialExpressionScalarParameter_28 = ScalarParameter() {defaultValue=0.1, group=SoilHaunt, parameterName=distortion_strength}
  MaterialExpressionDivide_3 = Divide(B=@MaterialExpressionConstant_24)
  n47 = Custom!(
   uv=@MaterialExpressionMultiply_33
   viewDir=@MaterialExpressionTransform_2
   steps=@MaterialExpressionConstant_24
   stepdist=@MaterialExpressionScalarParameter_24
   tex=@MaterialExpressionTextureObject_1
   vertnormal=@MaterialExpressionVertexNormalWS_4
   dTex=@MaterialExpressionTextureObject_1
   Time=@MaterialExpressionTime_4
   heightspeed=@MaterialExpressionComponentMask_2
   distortspeed=@MaterialExpressionComponentMask_3
   distortionstrength=@MaterialExpressionScalarParameter_28
   StepSize=@MaterialExpressionDivide_3
  ) {
   code="float3 rayStep = viewDir * -1;\r\nfloat2 uv2 = uv + Time*heightspeed;\r\nuv += Time*distortspeed;\r\nfloat4 heightTex = Texture2DSample(tex,texSampler,uv);\r\nfloat4 distortionTex = Texture2DSample(dTex,texSampler,uv2);\r\n\r\nfor(int i=0;i<=int(steps);i++)\r\n{\r\nif(heightTex.r<=i/steps)\r\n{\r\nreturn float((heightTex.r*i)/16.0)*1.0; \r\n}\r\nfloat stepct = heightTex.r * stepdist;\r\nuv += rayStep * stepct;\r\nuv2 += rayStep * stepct;\r\ndistortionTex = Texture2DSample(dTex,texSampler,uv2*1.0);\r\nfloat2 distortStr = distortionTex.xy * distortionstrength;\r\nheightTex = Texture2DSample(tex,texSampler,uv.xy + distortStr);\r\nRayMarchUV = uv.xy + distortStr;\r\n}\r\n\r\nreturn heightTex;"
   description=RayMarchTexture2
   outputType=CMOT_Float1
   showCode=true
   additionalOutputs [{outputName=RayMarchUV, outputType=CMOT_Float2}]
   inputs [
    {inputName=uv, input {expression=&self:MaterialExpressionMultiply_33}}
    {
     inputName=viewDir
     input {expression=&self:MaterialExpressionTransform_2}
    }
    {
     inputName=steps
     input {expression=&self:MaterialExpressionConstant_24}
    }
    {
     inputName=stepdist
     input {expression=&self:MaterialExpressionScalarParameter_24}
    }
    {
     inputName=tex
     input {expression=&self:MaterialExpressionTextureObject_1}
    }
    {
     inputName=vertnormal
     input {expression=&self:MaterialExpressionVertexNormalWS_4}
    }
    {
     inputName=dTex
     input {expression=&self:MaterialExpressionTextureObject_1}
    }
    {inputName=Time, input {expression=&self:MaterialExpressionTime_4}}
    {
     inputName=heightspeed
     input {expression=&self:MaterialExpressionComponentMask_2}
    }
    {
     inputName=distortspeed
     input {expression=&self:MaterialExpressionComponentMask_3}
    }
    {
     inputName=distortionstrength
     input {expression=&self:MaterialExpressionScalarParameter_28}
    }
    {
     inputName=StepSize
     input {expression=&self:MaterialExpressionDivide_3}
    }
   ]
  }
  NamedRerouteDeclaration(None=@n47.return) {name=RayMarchColor_output, nodeColor=<rgba 0.2422 0 1 1>}
  n48 = TransformPosition(None=WorldPosition().XYZ) {transformSourceType=TRANSFORMPOSSOURCE_World, transformType=TRANSFORMPOSSOURCE_View}
  n49 = VertexNormalWS()
  n50 = MaterialFunctionCall(Color=CrossProduct(A=Normalize(VectorInput=@n48), B=Transform(None=@n49) {transformSourceType=TRANSFORMSOURCE_World, transformType=TRANSFORM_View})) {materialFunction=&SplitComponents}
  n51 = Multiply(
   A=MaterialFunctionCall(R=@n50.G, G=@n50.G) {materialFunction=&AppendMany}.RG
   B=Constant2Vector() {g=0.5, r=0.5}
  )
  NamedRerouteDeclaration(None=ComponentMask(None=Add(A=@n51) {constB=0.5}) {g=true, r=true}) {name=MatCapUV, nodeColor=<rgba 1 0 0.2578 1>}
  n52 = NamedRerouteUsage()
  n53 = Constant() {r=1.13}
  n54 = Multiply(
   A=TextureSample(UVs=MaterialFunctionCall(UVs=@n52, Texture Scale=@n53) {materialFunction=&ScaleUVsByCenter}.UVs) {texture=&T_Fake_refraction}.R
   B=VectorParameter() {group=SoilHaunt, parameterName=Bright_Color, defaultValue=<rgba 0.1468 1 0.7473 1>}.RGB
  )
  NamedRerouteDeclaration(None=Add(A=Multiply(A=@n54) {constB=0.25}, B=@n13)) {name=Emissive Color, nodeColor=<rgba 1 0 0.1875 1>}
  n55 = Multiply(A=MaterialFunctionCall(Brightness=Constant() {r=3}, Exponent=Constant() {r=2}) {materialFunction=&MF_Fresnel_FX}.Fresnel)
  NamedRerouteDeclaration(None=Power(Base=Saturate(None=@n55))) {name=fresnel2, nodeColor=<rgba 0 0.2734 1 1>}
  n56 = NamedRerouteUsage()
  n57 = NamedRerouteUsage()
  n58 = Add(A=@n56, B=Multiply(B=@n57) {constA=-30})
  n59 = CameraPositionWS()
  n60 = NamedRerouteDeclaration(None=Add(
    A=VectorParameter() {group=SoilHaunt, parameterName=VP_PlaneNormal, defaultValue=<rgba 0 -1 0 1>}.RGB
    B=VectorParameter() {group=SoilHaunt, parameterName=VP_PlaneNormal_Offset, defaultValue=<rgba 0 0 0 1>}.RGB
   )) {name=VP_PlaneNormal_WS, nodeColor=<rgba 1 0.3516 0 1>}
  n61 = DotProduct(A=Normalize(VectorInput=Subtract(A=@n58, B=@n59)), B=Normalize(VectorInput=@n60))
  n62 = Saturate(None=OneMinus(None=Power(Base=@n61) {constExponent=1.5}))
  NamedRerouteDeclaration(None=LinearInterpolate(
    B=NamedRerouteDeclaration(None=@n62) {name=VirtualPlaneMask, nodeColor=<rgba 1 0.1875 0 1>}
    Alpha=ScalarParameter() {group=SoilHaunt, parameterName=Style01_weight}
   ) {constA=1}) {name=SkullSphereMask, nodeColor=<rgba 1 0.4688 0 1>}
  n63 = Add(A=@n56, B=Multiply(A=ScalarParameter() {defaultValue=-80, group=SoilHaunt, parameterName=CenterOffset}, B=@n57))
  NamedRerouteDeclaration(None=MaterialFunctionCall(
    Plane Normal Axis=@n60
    Plane Center=NamedRerouteDeclaration(None=@n63) {name=VP_PlaneCenter_WS, nodeColor=<rgba 1 0 0.5391 1>}
    UV Scale=ScalarParameter() {defaultValue=300, group=SoilHaunt, parameterName=VP_Scale}
   ) {materialFunction=&VirtualPlaneCoordinates}.UVs) {name=VirtualPlaneUV, nodeColor=<rgba 1 0 0.9375 1>}
  n64 = VertexNormalWS()
  n65 = Distance(A=VertexInterpolator(VS=PreSkinnedPosition()).PS, B=Constant())
  NamedRerouteDeclaration(None=Subtract(A=WorldPosition().XYZ, B=Multiply(A=@n64, B=@n65))) {name=SphereCenter, nodeColor=<rgba 0 0.3438 1 1>}
  NamedRerouteDeclaration(None=SphereMask(A=@n27.Linear Distance, B=Constant()) {attenuationRadius=0.35, hardnessPercent=0}) {name=VP_Spheremask, nodeColor=<rgba 1 0 0.4688 1>}
  Reroute(None=@n63)
 ]
}
